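# Lit material for the small bullet mesh particle in the WaltzScout pickaxe effects, usable with Niagara and Nanite.
# A specular texture drives roughness between two parameters and scales metallic, alongside a normal map.
# Emissive comes from a second-UV vertical power gradient scaled by the dynamic parameter's R channel as greyscale.
# That glow fades to black early in particle life through scaled relative time, then passes through MF_FX_EmissiveAdjustments.
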Material /BRCosmetics/Weapons/FORT_Melee/Pickaxe_WaltzScout/FX/M_Pickaxe_WaltzScout_BulletSmall
graph {
 properties {bUsedWithNanite=true, bUsedWithNiagaraMeshParticles=true, usedShadingModels=Default Lit}
 nodes [
  n1 = TextureSample() {texture=&/T_Pickaxe_WaltzScout_BulletSmall_S}
  LinearInterpolate(
   A=ScalarParameter() {defaultValue=0.22, group=Specular, parameterName=SpecRoughnessMin, sliderMax=1}
   B=ScalarParameter() {defaultValue=0.66, group=Specular, parameterName=SpecRoughnessMax, sliderMax=1}
   Alpha=@n1.B
  )
  n2 = TextureCoordinate() {coordinateIndex=1}
  n3 = Multiply(
   A=LinearInterpolate(Alpha=TextureSample() {texture=&/T_Pickaxe_WaltzScout_BulletSmall_D}.RGB) {constA=1}
   B=Power(Base=ComponentMask(None=@n2) {g=true}, Exp=ScalarParameter() {defaultValue=2, parameterName=EmissiveGradientExponent})
  )
  n4 = DynamicParameter() {defaultValue=<rgba 0 0 0 1>, paramNames [R,G,B,Param4]}
  n5 = MaterialFunctionCall(R=@n4.R, G=@n4.R, B=@n4.R) {materialFunction=&AppendMany}
  n6 = Multiply(A=ScalarParameter() {defaultValue=3, parameterName=PRT Scale}, B=ParticleRelativeTime())
  MaterialFunctionCall(Emissive=LinearInterpolate(A=Multiply(A=@n3, B=@n5.RGB), Alpha=Clamp(None=@n6)) {constB=0}) {materialFunction=&MF_FX_EmissiveAdjustments}
  Multiply(A=@n1.G, B=ScalarParameter() {defaultValue=1, parameterName=Metallic})
  TextureSample() {samplerType=SAMPLERTYPE_Normal, texture=&/T_Pickaxe_WaltzScout_BulletSmall_N}
 ]
}
