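# Animation Blueprint for the WaltzScout pickaxe that adds secondary physics motion to its four-bone dangling lace.
# The cached input pose feeds two Anim Dynamics chains with angular limits and a planar limit on the second bone.
# Blend by bIsInFrontEnd picks the world-space, heavy-gravity chain in the front end, otherwise the root-relative, lighter-gravity chain.
# The event graph computes airborne gravity and left/right velocity alphas, which the shown AnimGraph never reads.
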AnimBlueprint /BRCosmetics/Weapons/FORT_Melee/Pickaxe_WaltzScout/Meshes/WaltzScout_Axe_AnimBP
target_skeleton=&/WaltzScout_Axe_Skeleton
graphs [
 AnimGraph {nodes [
   n1 = Anim Dynamics - Bone: dyn_lace_1(ComponentPose=Local To Component(LocalPose=Use cached pose 'INPUT CACHE'()), Alpha=1, ExternalForce="0,0,0") {
    angularDampingOverride=0.9999
    bOverrideAngularDamping=true
    bUseGravityOverride=true
    simulationSpace=World
    boundBone {boneName=dyn_lace_1}
    componentLinearAccScale=<0.1 0.1 0.1>
    componentLinearVelScale=<0.1 0.1 0.1>
    gravityOverride {z=-4000}
    physicsBodyDefinitions [{
      boundBone {boneName=dyn_lace_1}
      boxExtents=<16 16 16>
      constraintSetup {angularLimitsMax=<0 75 0>, angularLimitsMin=<0 -25 0>, angularTarget=<0 0 0>}
      localJointOffset=<1 0 0>
     }]
    relativeSpaceBone {boneName=b_AR_Root}
   }
   n2 = Anim Dynamics - Bone: dyn_lace_2(ComponentPose=@n1, Alpha=1, ExternalForce="0,0,0") {
    angularDampingOverride=0.9999
    bOverrideAngularDamping=true
    bUseGravityOverride=true
    simulationSpace=World
    boundBone {boneName=dyn_lace_2}
    componentLinearAccScale=<0.1 0.1 0.1>
    componentLinearVelScale=<0.1 0.1 0.1>
    gravityOverride {z=-4000}
    physicsBodyDefinitions [{
      boundBone {boneName=dyn_lace_2}
      boxExtents=<16 16 16>
      constraintSetup {angularLimitsMax=<0 20 20>, angularLimitsMin=<0 -20 -20>, angularTarget=<0 0 0>}
      localJointOffset=<1 0 0>
     }]
    planarLimits [{drivingBone {boneName=b_AR_Root}, planeTransform {location=<-6 0 -8>, rotation=<rot 10 0 180>}}]
    relativeSpaceBone {boneName=b_AR_Root}
   }
   n3 = Anim Dynamics - Bone: dyn_lace_3(ComponentPose=@n2, Alpha=1, ExternalForce="0,0,0") {
    angularDampingOverride=0.9999
    bOverrideAngularDamping=true
    bUseGravityOverride=true
    simulationSpace=World
    boundBone {boneName=dyn_lace_3}
    componentLinearAccScale=<0.1 0.1 0.1>
    componentLinearVelScale=<0.1 0.1 0.1>
    gravityOverride {z=-4000}
    physicsBodyDefinitions [{
      boundBone {boneName=dyn_lace_3}
      boxExtents=<16 16 16>
      constraintSetup {angularLimitsMax=<5 10 20>, angularLimitsMin=<-5 -20 -20>, angularTarget=<0 0 0>}
      localJointOffset=<1 0 0>
     }]
    relativeSpaceBone {boneName=b_AR_Root}
   }
   n4 = Anim Dynamics - Bone: dyn_lace_4(ComponentPose=@n3, Alpha=1, ExternalForce="0,0,0") {
    angularDampingOverride=0.9999
    bOverrideAngularDamping=true
    bUseGravityOverride=true
    simulationSpace=World
    boundBone {boneName=dyn_lace_4}
    componentLinearAccScale=<0.1 0.1 0.1>
    componentLinearVelScale=<0.1 0.1 0.1>
    gravityOverride {z=-4000}
    physicsBodyDefinitions [{
      boundBone {boneName=dyn_lace_4}
      boxExtents=<16 16 16>
      constraintSetup {angularLimitsMax=<10 10 20>, angularLimitsMin=<-10 -20 -20>, angularTarget=<0 0 0>}
      localJointOffset=<1 0 0>
     }]
    relativeSpaceBone {boneName=b_AR_Root}
   }
   n5 = Anim Dynamics - Bone: dyn_lace_1!(
    ComponentPose=Local To Component(LocalPose=Use cached pose 'INPUT CACHE'())
    Alpha=1
    GravityOverride:Vector="0,0,-600"
    ExternalForce="0,0,0"
   ) -> (Pose) {
    angularDampingOverride=0.999
    bOverrideAngularDamping=true
    bUseGravityOverride=true
    simulationSpace=BoneRelative
    boundBone {boneName=dyn_lace_1}
    componentLinearAccScale=<0.5 0.5 0.5>
    componentLinearVelScale=<0.5 0.5 0.5>
    gravityOverride {z=-600}
    physicsBodyDefinitions [{
      boundBone {boneName=dyn_lace_1}
      boxExtents=<10 10 10>
      constraintSetup {angularLimitsMax=<0 75 0>, angularLimitsMin=<0 -25 0>, angularTarget=<0 0 0>}
      localJointOffset=<1 0 0>
     }]
    relativeSpaceBone {boneName=b_AR_Root}
   }
   n6 = Anim Dynamics - Bone: dyn_lace_2!(ComponentPose=@n5, Alpha=1, GravityOverride:Vector="0,0,-600", ExternalForce="0,0,0") -> (Pose) {
    angularDampingOverride=0.999
    bOverrideAngularDamping=true
    bUseGravityOverride=true
    simulationSpace=BoneRelative
    boundBone {boneName=dyn_lace_2}
    componentLinearAccScale=<0.5 0.5 0.5>
    componentLinearVelScale=<0.5 0.5 0.5>
    gravityOverride {z=-600}
    physicsBodyDefinitions [{
      boundBone {boneName=dyn_lace_2}
      boxExtents=<10 10 10>
      constraintSetup {angularLimitsMax=<5 20 20>, angularLimitsMin=<-5 -20 -20>, angularTarget=<0 0 0>}
      localJointOffset=<1 0 0>
     }]
    planarLimits [{drivingBone {boneName=b_AR_Root}, planeTransform {location=<-6 0 -12>, rotation=<rot 10 0 -180>}}]
    relativeSpaceBone {boneName=b_AR_Root}
   }
   n7 = Anim Dynamics - Bone: dyn_lace_3!(ComponentPose=@n6, Alpha=1, GravityOverride:Vector="0,0,-600", ExternalForce="0,0,0") -> (Pose) {
    angularDampingOverride=0.999
    bOverrideAngularDamping=true
    bUseGravityOverride=true
    simulationSpace=BoneRelative
    boundBone {boneName=dyn_lace_3}
    componentLinearAccScale=<0.5 0.7 0.5>
    componentLinearVelScale=<0.5 0.7 0.5>
    gravityOverride {z=-600}
    physicsBodyDefinitions [{
      boundBone {boneName=dyn_lace_3}
      boxExtents=<10 10 10>
      constraintSetup {angularLimitsMax=<10 10 20>, angularLimitsMin=<-10 -20 -20>, angularTarget=<0 0 0>}
      localJointOffset=<2 0 0>
     }]
    relativeSpaceBone {boneName=b_AR_Root}
   }
   n8 = Anim Dynamics - Bone: dyn_lace_4!(ComponentPose=@n7, Alpha=1, GravityOverride:Vector="0,0,-600", ExternalForce="0,0,0") -> (Pose) {
    angularDampingOverride=0.999
    bOverrideAngularDamping=true
    bUseGravityOverride=true
    simulationSpace=BoneRelative
    boundBone {boneName=dyn_lace_4}
    componentLinearAccScale=<0.5 0.7 0.5>
    componentLinearVelScale=<0.5 0.7 0.5>
    gravityOverride {z=-600}
    physicsBodyDefinitions [{
      boundBone {boneName=dyn_lace_4}
      boxExtents=<10 10 10>
      constraintSetup {angularLimitsMax=<10 10 20>, angularLimitsMin=<-10 -20 -20>, angularTarget=<0 0 0>}
      localJointOffset=<2 0 0>
     }]
    relativeSpaceBone {boneName=b_AR_Root}
   }
   Output Pose(Result=Blend Poses by bool(BlendPose_0=Component To Local(ComponentPose=@n4), BlendPose_1=Component To Local(ComponentPose=@n8), bActiveValue=Get bIsInFrontEnd())) {layerGroup=None}
   AnimGraphNode_SubInput_0 = Input Pose()
   Save cached pose 'INPUT CACHE'(Pose=@AnimGraphNode_SubInput_0) {cacheName=INPUT CACHE}
  ]}
 EventGraph {nodes [
   Event BlueprintUpdateAnimation()
   > n1 = Sequence()
   > n2 = Branch(Condition=Get bIsFalling())
   > Set Custom Gravity(Custom Gravity:Vector="0,0,300") -> (Output_Get:Vector)
   Set Custom Gravity(execute=@n2.else, Custom Gravity="0,0,-980")
   n3 = TryGetPawnOwner()
   n4 = Dot Product!(
    A:Vector=GetActorRightVector!(self:Actor Object Reference=@n3) -> (ReturnValue:Vector)
    B:Vector=GetVelocity!(self:Actor Object Reference=@n3) -> (ReturnValue:Vector)
   ) -> (ReturnValue:double)
   n5 = OR Boolean(A=Get bIsFalling(), B=Get bIsJumping())
   Set RightDirectionAlpha(
    execute=@n1.then_1
    RightDirectionAlpha:double=float * float(A=MapRangeClamped(Value=@n4, InRangeA=0, InRangeB=230), B:bool=@n5) {is=title K2Node_PromotableOperator}
   ) -> (Output_Get:double)
   > Set LeftDirectionAlpha(LeftDirectionAlpha:double=float * float(A=MapRangeClamped(Value=@n4, InRangeA=0, InRangeB=-230), B=@n5)) -> (Output_Get:double)
  ]}
]
variables [
 Custom Gravity {type=Vector}
 LeftDirectionAlpha {type=number}
 RightDirectionAlpha {type=number}
]
