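# Physics asset for the WaltzScout two-handed pickaxe cosmetic, defining sphere bodies for the handle and a three-bone dynamic lace.
# Constraints chain the lace bones in sequence from the handle body, suggesting a dangling secondary-motion strand.
# The root joint's limits give a wide swing on one axis and none on the other, while the lower joints have narrower swing limits and every joint has a minimal twist limit.
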PhysicsAsset /BRCosmetics/Weapons/FORT_Melee/Pickaxe_WaltzScout/Meshes/WaltzScout_Axe_Physics
bodies {
 Melee_TwoHanded_Handle [{shapeType=Sphere, radius=10.9578, shapeName=Melee_TwoHanded_Handle_sphere, center=<0 1.4681 -7.0356>}]
 dyn_lace_1 [{shapeType=Sphere, radius=3, shapeName=dyn_lace_1_sphere, center=<3 0 0>}]
 dyn_lace_2 [{shapeType=Sphere, radius=3, shapeName=dyn_lace_2_sphere, center=<3 0 0>}]
 dyn_lace_3 [{shapeType=Sphere, radius=3.5, shapeName=dyn_lace_3_sphere, center=<3 0 0>}]
}
constraints [
 {bone1Name=dyn_lace_1, bone2Name=Melee_TwoHanded_Handle, swing1LimitDegrees=0, swing2LimitDegrees=90, twistLimitDegrees=1}
 {bone1Name=dyn_lace_2, bone2Name=dyn_lace_1, swing1LimitDegrees=10, swing2LimitDegrees=20, twistLimitDegrees=1}
 {bone1Name=dyn_lace_3, bone2Name=dyn_lace_2, swing1LimitDegrees=10, swing2LimitDegrees=20, twistLimitDegrees=1}
]
